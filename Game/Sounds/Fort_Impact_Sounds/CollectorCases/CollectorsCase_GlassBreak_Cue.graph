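# Sound cue for the glass-break impact of a collector's case.
# A random node picks one of four glass-break waves with equal weight, and a modulator above it applies slight pitch variation.
# It uses the gadget-working attenuation asset, plays in the Athena ambient sound class, and has a boosted volume multiplier.
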SoundCue /Game/Sounds/Fort_Impact_Sounds/CollectorCases/CollectorsCase_GlassBreak_Cue
root=Modulator_0
nodes [
 {id=Modulator_0, children [Random_0], properties {pitchMax=1.1, pitchMin=0.9}}
 {id=Random_0, children [WavePlayer_0,WavePlayer_2,WavePlayer_3,WavePlayer_4], properties {weights [1,1,1,1]}}
 {id=WavePlayer_0, properties {soundWaveAssetPtr=&/CollectorsCase_GlassBreak_01}}
 {id=WavePlayer_2, properties {soundWaveAssetPtr=&/CollectorsCase_GlassBreak_02}}
 {id=WavePlayer_3, properties {soundWaveAssetPtr=&/CollectorsCase_GlassBreak_03}}
 {id=WavePlayer_4, properties {soundWaveAssetPtr=&/CollectorsCase_GlassBreak_04}}
]
properties {
 attenuationSettings=/Game/Sounds/Attenuation/Fort_GadgetWorking_Att
 maxDistance=3100
 soundClassObject=&Fort_Ambient_Athena
 volumeMultiplier=2.495
}
